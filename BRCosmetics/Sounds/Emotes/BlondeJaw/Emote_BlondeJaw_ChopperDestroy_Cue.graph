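# Sound cue for the chopper-destroy moment of the BlondeJaw emote.
# A random node picks one of three equally weighted destroy wave variants, then passes it through a modulator with pitch fixed at normal.
# It plays with the shared Athena emote attenuation, in the emote music sound class, at a slightly boosted volume.
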SoundCue /BRCosmetics/Sounds/Emotes/BlondeJaw/Emote_BlondeJaw_ChopperDestroy_Cue
root=Modulator_0
nodes [
 {id=Modulator_0, children [Random_0], properties {pitchMax=1, pitchMin=1}}
 {id=Random_0, children [WavePlayer_3,WavePlayer_4,WavePlayer_5], properties {weights [1,1,1]}}
 {id=WavePlayer_3, properties {soundWaveAssetPtr=&/IP_Emote_BlondeJaw_Chopper_Destroy_03}}
 {id=WavePlayer_4, properties {soundWaveAssetPtr=&/IP_Emote_BlondeJaw_Chopper_Destroy_02}}
 {id=WavePlayer_5, properties {soundWaveAssetPtr=&/IP_Emote_BlondeJaw_Chopper_Destroy_01}}
]
properties {
 attenuationSettings=&Fort_Emotes_Athena_Att
 maxDistance=1600
 soundClassObject=&Fort_Music_Emote
 volumeMultiplier=1.2
}
