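# Sound cue for the Kong roar in the BlondeJaw emote.
# A random node picks one of three Kong wave variations with equal weight, then passes it through a modulator.
# It plays through the Athena emote attenuation and the emote music sound class, with a slight volume boost.
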SoundCue /BRCosmetics/Sounds/Emotes/BlondeJaw/Emote_BlondeJaw_KongRoar_Cue
root=Modulator_0
nodes [
 {id=Modulator_0, children [Random_0]}
 {id=Random_0, children [WavePlayer_0,WavePlayer_1,WavePlayer_2], properties {weights [1,1,1]}}
 {id=WavePlayer_0, properties {soundWaveAssetPtr=&/IP_Emote_BlondeJaw_Kong_03}}
 {id=WavePlayer_1, properties {soundWaveAssetPtr=&/IP_Emote_BlondeJaw_Kong_02}}
 {id=WavePlayer_2, properties {soundWaveAssetPtr=&/IP_Emote_BlondeJaw_Kong_01}}
]
properties {
 attenuationSettings=&Fort_Emotes_Athena_Att
 maxDistance=1600
 soundClassObject=&Fort_Music_Emote
 volumeMultiplier=1.1
}
