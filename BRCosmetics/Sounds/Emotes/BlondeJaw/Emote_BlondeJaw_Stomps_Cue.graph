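# Sound cue for the stomp sounds of the BlondeJaw emote.
# A Random node picks among thirteen equally weighted stomp wave variants, feeding a Modulator that varies volume and pitch to reduce repetition.
# It plays at reduced volume through the shared Athena emote attenuation and is assigned to the emote music sound class.
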SoundCue /BRCosmetics/Sounds/Emotes/BlondeJaw/Emote_BlondeJaw_Stomps_Cue
root=Modulator_0
nodes [
 {id=Modulator_0, children [Random_0], properties {pitchMin=1, volumeMax=1.1, volumeMin=0.9}}
 {
  id=Random_0
  children [
   WavePlayer_0
   WavePlayer_1
   WavePlayer_2
   WavePlayer_6
   WavePlayer_7
   WavePlayer_8
   WavePlayer_9
   WavePlayer_10
   WavePlayer_11
   WavePlayer_12
   WavePlayer_13
   WavePlayer_14
   WavePlayer_15
  ]
  properties {weights [1,1,1,1,1,1,1,1,1,1,1,1,1]}
 }
 {id=WavePlayer_0, properties {soundWaveAssetPtr=&/IP_Emote_BlondeJaw_Stomp_01}}
 {id=WavePlayer_1, properties {soundWaveAssetPtr=&/IP_Emote_BlondeJaw_Stomp_02}}
 {id=WavePlayer_10, properties {soundWaveAssetPtr=&/IP_Emote_BlondeJaw_Stomp_08}}
 {id=WavePlayer_11, properties {soundWaveAssetPtr=&/IP_Emote_BlondeJaw_Stomp_09}}
 {id=WavePlayer_12, properties {soundWaveAssetPtr=&/IP_Emote_BlondeJaw_Stomp_10}}
 {id=WavePlayer_13, properties {soundWaveAssetPtr=&/IP_Emote_BlondeJaw_Stomp_11}}
 {id=WavePlayer_14, properties {soundWaveAssetPtr=&/IP_Emote_BlondeJaw_Stomp_12}}
 {id=WavePlayer_15, properties {soundWaveAssetPtr=&/IP_Emote_BlondeJaw_Stomp_13}}
 {id=WavePlayer_2, properties {soundWaveAssetPtr=&/IP_Emote_BlondeJaw_Stomp_03}}
 {id=WavePlayer_6, properties {soundWaveAssetPtr=&/IP_Emote_BlondeJaw_Stomp_04}}
 {id=WavePlayer_7, properties {soundWaveAssetPtr=&/IP_Emote_BlondeJaw_Stomp_05}}
 {id=WavePlayer_8, properties {soundWaveAssetPtr=&/IP_Emote_BlondeJaw_Stomp_06}}
 {id=WavePlayer_9, properties {soundWaveAssetPtr=&/IP_Emote_BlondeJaw_Stomp_07}}
]
properties {
 attenuationSettings=&Fort_Emotes_Athena_Att
 maxDistance=1600
 soundClassObject=&Fort_Music_Emote
 volumeMultiplier=0.7
}
